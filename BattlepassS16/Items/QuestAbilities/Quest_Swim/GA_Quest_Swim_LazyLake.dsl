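; Quest gameplay ability for the Battle Pass S16 Lazy Lake swim quest, crediting progress when the player swims at that location.
; On activation it calls the parent, then waits for the swimming tag to be added.
; When that fires, it compares the owning pawn's tags against island and town POI tags.
; If either matches, it sends a quest-tagged stat event to the player's quest manager and ends the ability.
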(event Ability|EventActivateAbility
  (bind (_case_0_true _case_1_true) (GameplayTags|TagContainer|CompareTagAssetInterfacetoOtherTags (Variables|Default|GetT_POI_Island) (Variables|Default|GetT_POI_Town) (Variables|GAVars|Data|GetOwningPlayerPawn)))
  (|CallParentFunction)
  (Ability|Tasks|WaitGameplayTagAdd (Variables|Default|GetT_Swimming)
    (:Added
      (if (or _case_0_true _case_1_true)
        (Quest|SendComplexCustomStatEvent (Variables|GAVars|Data|GetOwningPlayerQuestManager) 0 (Variables|Default|GetT_Quest) "(GameplayTags=)" 1)
        (Ability|EndAbility)))))

(event Ability|EventFailedToActivatePassiveAbility
  (params (ActorInfo :type Gameplay%20Ability%20Actor%20Info%20Structure)))

(event Ability|EventOnEndAbility
  (params (bWasCancelled :type Boolean)))
